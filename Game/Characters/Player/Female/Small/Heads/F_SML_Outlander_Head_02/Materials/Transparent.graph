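# Translucent material named Transparent for the Female Small Outlander Head 02 player character head.
# It is set up for skeletal meshes with morph targets and mesh deformers, using Default Lit shading and per-vertex non-directional volumetric translucency lighting.
# The graph holds a single Constant node at its default value, with no textures or parameters.
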Material /Game/Characters/Player/Female/Small/Heads/F_SML_Outlander_Head_02/Materials/Transparent
graph {
 properties {
  bUsedWithMeshDeformer=true
  bUsedWithMorphTargets=true
  bUsedWithSkeletalMesh=true
  blendMode=BLEND_Translucent
  pixelDepthOffsetMode=PDOM_Legacy
  translucencyLightingMode=TLM_VolumetricPerVertexNonDirectional
  usedShadingModels=Default Lit
 }
 nodes [Constant()]
}
